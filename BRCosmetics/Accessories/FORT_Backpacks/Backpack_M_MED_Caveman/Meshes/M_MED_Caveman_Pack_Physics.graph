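# Physics asset for the Caveman backpack cosmetic, providing bodies and joints for its dangling skull, two knives, and left and right bags.
# Small master capsules attach each group to the two-capsule pack body with free swing and twist.
# Child bodies hang from these anchors with angular limits, with narrower twist on the knives.
# Each bag forms a chain with a branching top piece, the right one longer and ending in a sphere.
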PhysicsAsset /BRCosmetics/Accessories/FORT_Backpacks/Backpack_M_MED_Caveman/Meshes/M_MED_Caveman_Pack_Physics
bodies {
 dyn_knife_01 [{shapeType=Capsule, length=6.3008, radius=0.9379, center=<4.4038 0.3816 0.7055>, rotation=<rot 90 -90 -94.9863>}]
 dyn_knife_01_MSTR [{shapeType=Capsule, length=0.505, radius=0.505}]
 dyn_knife_02 [{shapeType=Capsule, length=5.8608, radius=1.1221, center=<4.8299 0 0>, rotation=<rot 90 0 0>}]
 dyn_knife_02_MSTR [{shapeType=Capsule, length=0.505, radius=0.505}]
 dyn_l_bag01 [{shapeType=Capsule, length=2.8395, radius=2.2473, center=<1.9614 0 0>, rotation=<rot 90 0 0>}]
 dyn_l_bag02 [{shapeType=Capsule, length=5.1857, radius=4.0806, center=<4.9106 0 0>, rotation=<rot 90 0 0>}]
 dyn_l_bagTop [{shapeType=Capsule, length=2.3524, radius=1.9441, center=<2.0779 0 0>, rotation=<rot 90 0 0>}]
 dyn_l_bag_MSTR [{shapeType=Capsule, length=0.505, radius=0.505}]
 dyn_r_bag01 [{shapeType=Capsule, length=3.878, radius=2.0794, center=<3.8214 0 0>, rotation=<rot 90 0 0>}]
 dyn_r_bag02 [{shapeType=Capsule, length=2.6822, radius=2.6728, center=<3.927 0 0>, rotation=<rot 90 0 0>}]
 dyn_r_bag03 [{shapeType=Sphere, radius=3.1416, center=<5.1205 -0.1993 0.1329>}]
 dyn_r_bagTop [{shapeType=Capsule, length=1.8387, radius=1.3732, center=<1.0745 0 0>, rotation=<rot 90 0 0>}]
 dyn_r_bag_MSTR [{shapeType=Capsule, length=0.505, radius=0.505}]
 dyn_skull [{shapeType=Capsule, length=4.5533, radius=3.6006, center=<3.6043 0 0.0001>, rotation=<rot 90 0 0>}]
 dyn_skull_MSTR [{shapeType=Capsule, length=0.505, radius=0.505}]
 pack [
  {shapeType=Capsule, length=15.3129, radius=9.8838, center=<-1.7235 -6.3353 -5.6624>}
  {shapeType=Capsule, length=14.9919, radius=10.2048, center=<1.6597 -6.6129 -5.6624>}
 ]
}
constraints [
 {
  bone1Name=dyn_skull_MSTR
  bone2Name=pack
  swing1LimitDegrees=45
  swing1Motion=Free
  swing2LimitDegrees=45
  swing2Motion=Free
  twistLimitDegrees=45
  twistMotion=Free
 }
 {
  bone1Name=dyn_knife_01_MSTR
  bone2Name=pack
  swing1LimitDegrees=45
  swing1Motion=Free
  swing2LimitDegrees=45
  swing2Motion=Free
  twistLimitDegrees=45
  twistMotion=Free
 }
 {
  bone1Name=dyn_knife_02_MSTR
  bone2Name=pack
  swing1LimitDegrees=45
  swing1Motion=Free
  swing2LimitDegrees=45
  swing2Motion=Free
  twistLimitDegrees=45
  twistMotion=Free
 }
 {
  bone1Name=dyn_l_bag_MSTR
  bone2Name=pack
  swing1LimitDegrees=45
  swing1Motion=Free
  swing2LimitDegrees=45
  swing2Motion=Free
  twistLimitDegrees=45
  twistMotion=Free
 }
 {
  bone1Name=dyn_r_bag_MSTR
  bone2Name=pack
  swing1LimitDegrees=45
  swing1Motion=Free
  swing2LimitDegrees=45
  swing2Motion=Free
  twistLimitDegrees=45
  twistMotion=Free
 }
 {bone1Name=dyn_skull, bone2Name=dyn_skull_MSTR, swing1LimitDegrees=45, swing2LimitDegrees=45, twistLimitDegrees=45}
 {bone1Name=dyn_knife_01, bone2Name=dyn_knife_01_MSTR, swing1LimitDegrees=80, swing2LimitDegrees=45, twistLimitDegrees=10}
 {bone1Name=dyn_knife_02, bone2Name=dyn_knife_02_MSTR, swing1LimitDegrees=80, swing2LimitDegrees=45, twistLimitDegrees=10}
 {bone1Name=dyn_l_bag01, bone2Name=dyn_l_bag_MSTR, swing1LimitDegrees=45, swing2LimitDegrees=45, twistLimitDegrees=45}
 {bone1Name=dyn_l_bag02, bone2Name=dyn_l_bag01, swing1LimitDegrees=45, swing2LimitDegrees=45, twistLimitDegrees=45}
 {bone1Name=dyn_l_bagTop, bone2Name=dyn_l_bag01, swing1LimitDegrees=45, swing2LimitDegrees=45, twistLimitDegrees=45}
 {bone1Name=dyn_r_bag01, bone2Name=dyn_r_bag_MSTR, swing1LimitDegrees=45, swing2LimitDegrees=45, twistLimitDegrees=45}
 {bone1Name=dyn_r_bag02, bone2Name=dyn_r_bag01, swing1LimitDegrees=45, swing2LimitDegrees=45, twistLimitDegrees=45}
 {bone1Name=dyn_r_bag03, bone2Name=dyn_r_bag02, swing1LimitDegrees=45, swing2LimitDegrees=45, twistLimitDegrees=45}
 {bone1Name=dyn_r_bagTop, bone2Name=dyn_r_bag02, swing1LimitDegrees=45, swing2LimitDegrees=45, twistLimitDegrees=45}
]
